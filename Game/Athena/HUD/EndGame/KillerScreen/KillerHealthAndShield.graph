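# End-game killer screen widget that shows the killer's health and shield as stacked material-driven bars with icons and numeric labels.
# Set Health And Shield makes it visible, writes both fractions to each bar's Progress material parameter, and displays them times 100 as whole numbers.
# HUD-tag wrappers and a root CommonTagVisibilityWidget, initialized on Construct, support tag-based hiding.
# PreConstruct applies the Show Icons toggle and text color.
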WidgetBlueprint /Game/Athena/HUD/EndGame/KillerScreen/KillerHealthAndShield
graphs [EventGraph {nodes [
   n1 = Set Health And Shield!() -> (OutputDelegate, then, Health PCT:double=~, Shield PCT:double=~) {is=title K2Node_CustomEvent}
   > n2 = Sequence()
   > SetVisibility(InVisibility=HitTestInvisible)
   > n3 = GetDynamicMaterial(self=Get Health_Bar() -> (Health_Bar:Image Object Reference))
   > SetScalarParameterValue(self=@n3, ParameterName=Progress, Value=@n1.Health PCT)
   > n4 = GetDynamicMaterial(self=Get Shield_Bar() -> (Shield_Bar:Image Object Reference))
   > SetScalarParameterValue(self=@n4, ParameterName=Progress, Value=@n1.Shield PCT)
   "SetText (Text)"(
    execute=@n2.then_1
    self=Get CommonText_Health() -> (CommonText_Health:Common Text Object Reference)
    InText="To Text (Float)"(
     Value=float * float(A=@n1.Health PCT, B=100)
     MinimumIntegralDigits=1
     MaximumIntegralDigits=324
     MinimumFractionalDigits=0
     MaximumFractionalDigits=0
    )
   )
   > "SetText (Text)"(
    self=Get CommonText_Shields() -> (CommonText_Shields:Common Text Object Reference)
    InText="To Text (Float)"(
     Value=float * float(A=@n1.Shield PCT, B=100)
     MinimumIntegralDigits=1
     MaximumIntegralDigits=324
     MinimumFractionalDigits=0
     MaximumFractionalDigits=0
    )
   )
   Event Construct()
   > Initialize!(
    execute
    self:Common Tag Visibility Widget Object Reference=Get CommonTagVisibilityWidget() -> (CommonTagVisibilityWidget:Common Tag Visibility Widget Object Reference)
    OwningLocalPlayer:Local Player Object Reference=GetOwningLocalPlayer!(self:Widget Object Reference) -> (ReturnValue:Local Player Object Reference)
    OwningPlayerController:Player Controller Object Reference=GetOwningPlayer()
   ) -> (then)
   Event PreConstruct()
   > n5 = Sequence()
   n6 = Select(Option 0=Collapsed, Option 1=SelfHitTestInvisible, Index=Get Show Icons() -> (Show Icons:bool))
   > SetVisibility(self=Get Icon_Shield() -> (Icon_Shield:Image Object Reference), InVisibility=@n6)
   > SetVisibility(self=Get Icon_Health() -> (Icon_Health:Image Object Reference), InVisibility=@n6)
   n7 = Get Text Color and Opacity() -> (Text Color and Opacity:Slate Color Structure)
   SetColorAndOpacity(execute=@n5.then_1, self=Get CommonText_Shields(), InColorAndOpacity=@n7)
   > SetColorAndOpacity(self=Get CommonText_Health(), InColorAndOpacity=@n7)
  ]}]
variables [
 Show Icons {type=boolean, default=true}
 Text Color and Opacity {type=SlateColor, default {colorUseRule=UseColor_Specified, specifiedColor=<rgba 1 1 1 1>}}
]
widgets {
 info {
  inheritedWidgetCount=0
  namedSlotCount=2
  parentClass=&UserWidget
  rootWidgetClass=/Script/FortniteUI.CommonTagVisibilityWidget
  widgetCount=20
 }
 widgets [
  CommonTagVisibilityWidget {
   widgetClassPath=&CommonTagVisibilityWidget
   bIsVariable=true
   widget=CommonTagVisibilityWidget
   properties {hUDElementTag {gameplayTags [{tagName=HUD.Athena.KillerInfo.HealthAndShield}]}}
  }
  VerticalBox_14 {widgetClassPath=&VerticalBox, parent=CommonTagVisibilityWidget, widget=VerticalBox_14, slot {class=PanelSlot}}
  FortHUDElementWrapper_Shield {
   widgetClassPath=/Game/Athena/HUD/FortHUDElementWrapper
   bIsVariable=true
   parent=VerticalBox_14
   widget=FortHUDElementWrapper_Shield
   properties {
    elementTag {tagName=HUD.Athena.Player.Shield}
    hUDElementTag {gameplayTags [{tagName=Athena.HideForCapture}, {tagName=Athena.Creative.HUDController.HideShields}]}
   }
   slot {class=VerticalBoxSlot}
  }
  HorizontalBox_4 {widgetClassPath=&HorizontalBox, namedSlotHost=FortHUDElementWrapper_Shield, widget=HorizontalBox_4}
  IconFixedSpacing {
   widgetClassPath=&SizeBox
   bIsVariable=true
   parent=HorizontalBox_4
   widget=IconFixedSpacing
   properties {bOverride_HeightOverride=true, bOverride_WidthOverride=true, heightOverride=24, widthOverride=24}
   slot {class=HorizontalBoxSlot, properties {horizontalAlignment=HAlign_Center, verticalAlignment=VAlign_Center, padding=<ltrb -12 -12 0 -12>}}
  }
  IconScale {
   widgetClassPath=&ScaleBox
   bIsVariable=true
   parent=IconFixedSpacing
   widget=IconScale
   properties {stretch=UserSpecified}
   slot {class=SizeBoxSlot}
  }
  Icon_Shield {
   widgetClassPath=&Image
   bIsVariable=true
   parent=IconScale
   widget=Icon_Shield
   properties {brush {
     bIsDynamicallyLoaded=false
     drawAs=Image
     imageType=NoImage
     mirroring=NoMirror
     resourceName=None
     resourceObject=/Game/Athena/HUD/Materials/M_UI_HitPoint_Type_Shield
     tiling=NoTile
     imageSize=<24 24>
     margin=<ltrb 0 0 0 0>
     outlineSettings {
      bUseBrushTransparency=false
      roundingType=HalfHeightRadius
      width=0
      color {colorUseRule=UseColor_Specified, specifiedColor=<rgba 0 0 0 0>}
      cornerRadii=<0 0 0 0>
     }
     tintColor {colorUseRule=UseColor_Specified, specifiedColor=<rgba 1 1 1 1>}
     uVRegion {bIsValid=false, max=<0 0>, min=<0 0>}
    }}
   slot {class=ScaleBoxSlot}
  }
  SizeBox_1 {
   widgetClassPath=&SizeBox
   parent=HorizontalBox_4
   widget=SizeBox_1
   properties {bOverride_HeightOverride=true, bOverride_WidthOverride=true, heightOverride=16, widthOverride=350}
   slot {class=HorizontalBoxSlot, properties {horizontalAlignment=HAlign_Left, verticalAlignment=VAlign_Center}}
  }
  Overlay_0 {widgetClassPath=&Overlay, parent=SizeBox_1, widget=Overlay_0, slot {class=SizeBoxSlot}}
  Shield_Bar {
   widgetClassPath=&Image
   bIsVariable=true
   parent=Overlay_0
   widget=Shield_Bar
   properties {brush {
     bIsDynamicallyLoaded=false
     drawAs=Image
     imageType=NoImage
     mirroring=NoMirror
     resourceName=None
     resourceObject=/Game/Athena/HUD/Materials/M_UI_HitPointBar_Shield_WithBacking
     tiling=NoTile
     imageSize=<32 32>
     margin=<ltrb 0 0 0 0>
     outlineSettings {
      bUseBrushTransparency=false
      roundingType=HalfHeightRadius
      width=0
      color {colorUseRule=UseColor_Specified, specifiedColor=<rgba 0 0 0 0>}
      cornerRadii=<0 0 0 0>
     }
     tintColor {colorUseRule=UseColor_Specified, specifiedColor=<rgba 1 1 1 1>}
     uVRegion {bIsValid=false, max=<0 0>, min=<0 0>}
    }}
   slot {class=OverlaySlot, properties {horizontalAlignment=HAlign_Fill, verticalAlignment=VAlign_Fill}}
  }
  CommonText_Shields {
   widgetClassPath=&CommonTextBlock
   bIsVariable=true
   parent=HorizontalBox_4
   widget=CommonText_Shields
   properties {
    text=999
    font {
     bForceMonospaced=false
     bMaterialIsStencil=false
     fontMaterial=null
     fontObject=&BurbankSmall
     letterSpacing=0
     monospacedWidth=1
     size=15
     skewAmount=0
     typefaceFontName=Black
     outlineSettings {
      bApplyOutlineToDropShadows=true
      bMiteredCorners=false
      bSeparateFillAlpha=true
      outlineMaterial=&M-FontShadow
      outlineSize=2
      outlineColor=<rgba 0 0 0 0.5>
     }
    }
    shadowOffset=<0 0>
   }
   slot {class=HorizontalBoxSlot, properties {horizontalAlignment=HAlign_Left, verticalAlignment=VAlign_Center, padding=<ltrb 5 0 5 0>}}
  }
  FortHUDElementWrapper_Health {
   widgetClassPath=&FortHUDElementWrapper
   bIsVariable=true
   parent=VerticalBox_14
   widget=FortHUDElementWrapper_Health
   properties {
    elementTag {tagName=HUD.Athena.Player.Health}
    hUDElementTag {gameplayTags [{tagName=Athena.HideForCapture}, {tagName=Athena.Creative.HUDController.HideHealth}]}
   }
   slot {class=VerticalBoxSlot}
  }
  HorizontalBox_5 {widgetClassPath=&HorizontalBox, namedSlotHost=FortHUDElementWrapper_Health, widget=HorizontalBox_5}
  IconFixedSpacing_1 {
   widgetClassPath=&SizeBox
   bIsVariable=true
   parent=HorizontalBox_5
   widget=IconFixedSpacing_1
   properties {bOverride_HeightOverride=true, bOverride_WidthOverride=true, heightOverride=24, widthOverride=24}
   slot {class=HorizontalBoxSlot, properties {horizontalAlignment=HAlign_Center, verticalAlignment=VAlign_Center, padding=<ltrb -12 -12 0 -12>}}
  }
  IconScale_1 {
   widgetClassPath=&ScaleBox
   bIsVariable=true
   parent=IconFixedSpacing_1
   widget=IconScale_1
   properties {stretch=UserSpecified}
   slot {class=SizeBoxSlot}
  }
  Icon_Health {
   widgetClassPath=&Image
   bIsVariable=true
   parent=IconScale_1
   widget=Icon_Health
   properties {brush {
     bIsDynamicallyLoaded=false
     drawAs=Image
     imageType=NoImage
     mirroring=NoMirror
     resourceName=None
     resourceObject=/Game/Athena/HUD/Materials/M_UI_HitPoint_Type_Health
     tiling=NoTile
     imageSize=<24 24>
     margin=<ltrb 0 0 0 0>
     outlineSettings {
      bUseBrushTransparency=false
      roundingType=HalfHeightRadius
      width=0
      color {colorUseRule=UseColor_Specified, specifiedColor=<rgba 0 0 0 0>}
      cornerRadii=<0 0 0 0>
     }
     tintColor {colorUseRule=UseColor_Specified, specifiedColor=<rgba 1 1 1 1>}
     uVRegion {bIsValid=false, max=<0 0>, min=<0 0>}
    }}
   slot {class=ScaleBoxSlot}
  }
  SizeBox_2 {
   widgetClassPath=&SizeBox
   parent=HorizontalBox_5
   widget=SizeBox_2
   properties {bOverride_HeightOverride=true, bOverride_WidthOverride=true, heightOverride=32, widthOverride=350}
   slot {class=HorizontalBoxSlot, properties {horizontalAlignment=HAlign_Left}}
  }
  Overlay_1 {widgetClassPath=&Overlay, parent=SizeBox_2, widget=Overlay_1, slot {class=SizeBoxSlot}}
  Health_Bar {
   widgetClassPath=&Image
   bIsVariable=true
   parent=Overlay_1
   widget=Health_Bar
   properties {brush {
     bIsDynamicallyLoaded=false
     drawAs=Image
     imageType=NoImage
     mirroring=NoMirror
     resourceName=None
     resourceObject=/Game/Athena/HUD/Materials/M_UI_HitPointBar_Health_WithBacking
     tiling=NoTile
     imageSize=<32 32>
     margin=<ltrb 0 0 0 0>
     outlineSettings {
      bUseBrushTransparency=false
      roundingType=HalfHeightRadius
      width=0
      color {colorUseRule=UseColor_Specified, specifiedColor=<rgba 0 0 0 0>}
      cornerRadii=<0 0 0 0>
     }
     tintColor {colorUseRule=UseColor_Specified, specifiedColor=<rgba 1 1 1 1>}
     uVRegion {bIsValid=false, max=<0 0>, min=<0 0>}
    }}
   slot {class=OverlaySlot, properties {horizontalAlignment=HAlign_Fill, verticalAlignment=VAlign_Fill}}
  }
  CommonText_Health {
   widgetClassPath=&CommonTextBlock
   bIsVariable=true
   parent=HorizontalBox_5
   widget=CommonText_Health
   properties {
    text=999
    font {
     bForceMonospaced=false
     bMaterialIsStencil=false
     fontMaterial=null
     fontObject=&BurbankSmall
     letterSpacing=0
     monospacedWidth=1
     size=15
     skewAmount=0
     typefaceFontName=Black
     outlineSettings {
      bApplyOutlineToDropShadows=true
      bMiteredCorners=false
      bSeparateFillAlpha=true
      outlineMaterial=&M-FontShadow
      outlineSize=2
      outlineColor=<rgba 0 0 0 0.5>
     }
    }
    shadowOffset=<0 0>
   }
   slot {class=HorizontalBoxSlot, properties {horizontalAlignment=HAlign_Left, verticalAlignment=VAlign_Center, padding=<ltrb 5 0 5 0>}}
  }
 ]
}
