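# Killer-screen HUD widget that shows the killer's weapon icon, name and rarity, gated by a CommonTagVisibilityWidget tagged HUD.Athena.KillerInfo.WeaponInfo.
# Set Icon And Name shows the widget, fills icon, drop shadow and name, calls Colorize, then writes the rarity name in its color.
# Colorize feeds rarity colors into the background's dynamic material and tints two scrolling triangle images, whose parent overlay starts collapsed.
# PreConstruct applies Uncommon colors as a default.
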WidgetBlueprint /Game/Athena/HUD/EndGame/KillerScreen/KillerInfoWeaponWidget
graphs [EventGraph {nodes [
   n1 = Colorize!() -> (OutputDelegate, then, Rarity:EFortRarity Enum=Common) {is=title K2Node_CustomEvent}
   > n2 = GetRarityData!(execute, Rarity:EFortRarity Enum=@n1.Rarity) -> (then, ReturnValue:Fort Rarity Item Data Structure)
   > n3 = GetDynamicMaterial(self=Get Image_ItemBG() -> (Image_ItemBG:Image Object Reference))
   n4 = Break Fort Rarity Item Data!(FortRarityItemData:"Fort Rarity Item Data Structure (by ref)"=@n2) -> (
    Name:Text
    Color1:Linear Color Structure
    Color2:Linear Color Structure
    Color3:Linear Color Structure
    Color4:Linear Color Structure
    Color5:Linear Color Structure
    Radius:float=~
    Falloff:float=~
    Brightness:float=~
    Roughness:float=~
    Glow:float=~
   ) {is=title K2Node_BreakStruct}
   > SetVectorParameterValue(self=@n3, ParameterName=BGColor, Value=@n4.Color3)
   > SetVectorParameterValue(self=@n3, ParameterName=GradientColor, Value=@n4.Color2)
   > SetVectorParameterValue(self=@n3, ParameterName=OutlineColor, Value=@n4.Color1)
   > SetColorAndOpacity(
    self:Image Object Reference=[Get ImageTriangleL() -> (ImageTriangleL:Image Object Reference), Get ImageTriangleR() -> (ImageTriangleR:Image Object Reference)]
    InColorAndOpacity:Linear Color Structure=@n4.Color1
   )
   n5 = Set Icon And Name!() -> (OutputDelegate, then, Weapon Icon:Texture 2D Object Reference, Weapon Name:Text, Weapon Rarity:EFortRarity Enum=Common) {is=title K2Node_CustomEvent}
   > SetVisibility(InVisibility=HitTestInvisible)
   > SetBrushFromTexture!(
    execute
    self:Image Object Reference=[Get Image_Icon() -> (Image_Icon:Image Object Reference), Get Image_IconShadow() -> (Image_IconShadow:Image Object Reference)]
    Texture:Texture 2D Object Reference=@n5.Weapon Icon
    bMatchSize:bool=false
   ) -> (then)
   > "SetText (Text)"(self=Get Text_WeaponName() -> (Text_WeaponName:Common Text Object Reference), InText=@n5.Weapon Name)
   > Colorize!(execute, self, Rarity:EFortRarity Enum=@n5.Weapon Rarity) -> (then) {is=title K2Node_CallFunction}
   > n6 = GetRarityData(Rarity=@n5.Weapon Rarity)
   n7 = Break Fort Rarity Item Data(FortRarityItemData=@n6) -> (Radius=~, Falloff=~, Brightness=~, Roughness=~, Glow=~)
   n8 = Get Text_WeaponRarity() -> (Text_WeaponRarity:Common Text Object Reference)
   > "SetText (Text)"(self=@n8, InText=Format Text!(Format="LOCTEXT(\"({rarity})\")", rarity:Text=@n7) -> (Result))
   > SetColorAndOpacity(self:Text Object Reference=@n8, InColorAndOpacity:Slate Color Structure=Make Slate Color(SpecifiedColor=@n7.Color1))
   Event PreConstruct()
   > Colorize(Rarity=Uncommon)
   Event Construct()
   > Initialize!(
    execute
    self:Common Tag Visibility Widget Object Reference=Get CommonTagVisibilityWidget() -> (CommonTagVisibilityWidget:Common Tag Visibility Widget Object Reference)
    OwningLocalPlayer:Local Player Object Reference=GetOwningLocalPlayer!(self:Widget Object Reference) -> (ReturnValue:Local Player Object Reference)
    OwningPlayerController:Player Controller Object Reference=GetOwningPlayer()
   ) -> (then)
  ]}]
variables []
widgets {
 info {inheritedWidgetCount=0, parentClass=&UserWidget, rootWidgetClass=/Script/FortniteUI.CommonTagVisibilityWidget, widgetCount=22}
 widgets [
  CommonTagVisibilityWidget {
   widgetClassPath=&CommonTagVisibilityWidget
   bIsVariable=true
   widget=CommonTagVisibilityWidget
   properties {hUDElementTag {gameplayTags [{tagName=HUD.Athena.KillerInfo.WeaponInfo}]}}
  }
  Overlay_Container {widgetClassPath=&Overlay, parent=CommonTagVisibilityWidget, widget=Overlay_Container, slot {class=PanelSlot}}
  VerticalBox_0 {
   widgetClassPath=&VerticalBox
   parent=Overlay_Container
   widget=VerticalBox_0
   slot {class=OverlaySlot, properties {verticalAlignment=VAlign_Center}}
  }
  Overlay_2 {widgetClassPath=&Overlay, parent=VerticalBox_0, widget=Overlay_2, slot {class=VerticalBoxSlot}}
  Overlay_4 {
   widgetClassPath=&Overlay
   parent=Overlay_2
   widget=Overlay_4
   properties {visibility=Collapsed}
   slot {class=OverlaySlot, properties {horizontalAlignment=HAlign_Fill, verticalAlignment=VAlign_Fill}}
  }
  SizeBox_0 {
   widgetClassPath=&SizeBox
   parent=Overlay_4
   widget=SizeBox_0
   properties {bOverride_WidthOverride=true, widthOverride=128}
   slot {class=OverlaySlot, properties {verticalAlignment=VAlign_Center, padding=<ltrb 90 0 -500 0>}}
  }
  ImageTriangleR {
   widgetClassPath=&Image
   bIsVariable=true
   parent=SizeBox_0
   widget=ImageTriangleR
   properties {
    brush {
     bIsDynamicallyLoaded=false
     drawAs=Image
     imageType=NoImage
     mirroring=NoMirror
     resourceName=None
     resourceObject=/Game/UI/HUD/Materials/M_FX_TriangleScroll-Burst
     tiling=NoTile
     imageSize=<32 90>
     margin=<ltrb 0 0 0 0>
     outlineSettings {
      bUseBrushTransparency=false
      roundingType=HalfHeightRadius
      width=0
      color {colorUseRule=UseColor_Specified, specifiedColor=<rgba 0 0 0 0>}
      cornerRadii=<0 0 0 0>
     }
     tintColor {colorUseRule=UseColor_Specified, specifiedColor=<rgba 1 1 1 1>}
     uVRegion {bIsValid=false, max=<0 0>, min=<0 0>}
    }
    renderTransform {angle=0, scale=<1 -1>, shear=<0 0>, translation=<-7 0>}
    renderTransformPivot=<0 0.5>
   }
   slot {class=SizeBoxSlot}
  }
  SizeBox_1 {
   widgetClassPath=&SizeBox
   parent=Overlay_4
   widget=SizeBox_1
   properties {bOverride_WidthOverride=true, widthOverride=128, renderTransform {angle=0, scale=<-1 1>, shear=<0 0>, translation=<0 0>}}
   slot {class=OverlaySlot, properties {horizontalAlignment=HAlign_Right, verticalAlignment=VAlign_Center, padding=<ltrb -500 0 90 0>}}
  }
  ImageTriangleL {
   widgetClassPath=&Image
   bIsVariable=true
   parent=SizeBox_1
   widget=ImageTriangleL
   properties {brush {
     bIsDynamicallyLoaded=false
     drawAs=Image
     imageType=NoImage
     mirroring=NoMirror
     resourceName=None
     resourceObject=&M_FX_TriangleScroll-Burst
     tiling=NoTile
     imageSize=<32 90>
     margin=<ltrb 0 0 0 0>
     outlineSettings {
      bUseBrushTransparency=false
      roundingType=HalfHeightRadius
      width=0
      color {colorUseRule=UseColor_Specified, specifiedColor=<rgba 0 0 0 0>}
      cornerRadii=<0 0 0 0>
     }
     tintColor {colorUseRule=UseColor_Specified, specifiedColor=<rgba 1 1 1 1>}
     uVRegion {bIsValid=false, max=<0 0>, min=<0 0>}
    }}
   slot {class=SizeBoxSlot}
  }
  Overlay_0 {widgetClassPath=&Overlay, parent=Overlay_2, widget=Overlay_0, slot {class=OverlaySlot}}
  Image_84 {
   widgetClassPath=&Image
   bIsVariable=true
   parent=Overlay_0
   widget=Image_84
   properties {colorAndOpacity=<rgba 0.0052 0.0052 0.0056 0.2>, renderTransform {angle=0, scale=<1 1>, shear=<0 0>, translation=<8 8>}}
   slot {class=OverlaySlot, properties {horizontalAlignment=HAlign_Fill, verticalAlignment=VAlign_Fill}}
  }
  CommonBorder_0 {
   widgetClassPath=&CommonBorder
   parent=Overlay_0
   widget=CommonBorder_0
   properties {
    background {
     bIsDynamicallyLoaded=false
     drawAs=Image
     imageType=NoImage
     mirroring=NoMirror
     resourceName=None
     resourceObject=null
     tiling=NoTile
     imageSize=<32 32>
     margin=<ltrb 0 0 0 0>
     outlineSettings {
      bUseBrushTransparency=false
      roundingType=HalfHeightRadius
      width=0
      color {colorUseRule=UseColor_Specified, specifiedColor=<rgba 0 0 0 0>}
      cornerRadii=<0 0 0 0>
     }
     tintColor {colorUseRule=UseColor_Specified, specifiedColor=<rgba 1 1 1 1>}
     uVRegion {bIsValid=false, max=<0 0>, min=<0 0>}
    }
    padding=<ltrb 4 4 4 4>
   }
   slot {class=OverlaySlot}
  }
  Image_ItemBG {
   widgetClassPath=&Image
   bIsVariable=true
   parent=CommonBorder_0
   widget=Image_ItemBG
   properties {brush {
     bIsDynamicallyLoaded=false
     drawAs=Image
     imageType=NoImage
     mirroring=NoMirror
     resourceName=None
     resourceObject=/Game/Athena/Playlists/Goose/UI/Material/M_GooseItemBG_Inst
     tiling=NoTile
     imageSize=<100 100>
     margin=<ltrb 0 0 0 0>
     outlineSettings {
      bUseBrushTransparency=false
      roundingType=HalfHeightRadius
      width=0
      color {colorUseRule=UseColor_Specified, specifiedColor=<rgba 0 0 0 0>}
      cornerRadii=<0 0 0 0>
     }
     tintColor {colorUseRule=UseColor_Specified, specifiedColor=<rgba 1 1 1 1>}
     uVRegion {bIsValid=false, max=<0 0>, min=<0 0>}
    }}
   slot {class=BorderSlot, properties {padding=<ltrb 4 4 4 4>}}
  }
  Overlay_ItemIcon {
   widgetClassPath=&Overlay
   parent=Overlay_Container
   widget=Overlay_ItemIcon
   properties {renderTransform {angle=0, scale=<1 1>, shear=<0 0>, translation=<-12 0>}}
   slot {class=OverlaySlot, properties {verticalAlignment=VAlign_Center, padding=<ltrb 0 -45 0 -25>}}
  }
  Image_IconShadow {
   widgetClassPath=&Image
   bIsVariable=true
   parent=Overlay_ItemIcon
   widget=Image_IconShadow
   properties {
    brush {
     bIsDynamicallyLoaded=false
     drawAs=Image
     imageType=NoImage
     mirroring=NoMirror
     resourceName=None
     resourceObject=/Game/UI/Foundation/Textures/Icons/Weapons/Items/T-Icon-Weapons-SK-AutoHigh-Suppressed
     tiling=NoTile
     imageSize=<160 160>
     margin=<ltrb 0 0 0 0>
     outlineSettings {
      bUseBrushTransparency=false
      roundingType=HalfHeightRadius
      width=0
      color {colorUseRule=UseColor_Specified, specifiedColor=<rgba 0 0 0 0>}
      cornerRadii=<0 0 0 0>
     }
     tintColor {colorUseRule=UseColor_Specified, specifiedColor=<rgba 1 1 1 1>}
     uVRegion {bIsValid=false, max=<0 0>, min=<0 0>}
    }
    colorAndOpacity=<rgba 0.0052 0.0052 0.0056 0.3>
    renderTransform {angle=0, scale=<1 1>, shear=<0 0>, translation=<6 12>}
   }
   slot {class=OverlaySlot}
  }
  Image_Icon {
   widgetClassPath=&Image
   bIsVariable=true
   parent=Overlay_ItemIcon
   widget=Image_Icon
   properties {brush {
     bIsDynamicallyLoaded=false
     drawAs=Image
     imageType=NoImage
     mirroring=NoMirror
     resourceName=None
     resourceObject=&T-Icon-Weapons-SK-AutoHigh-Suppressed
     tiling=NoTile
     imageSize=<160 160>
     margin=<ltrb 0 0 0 0>
     outlineSettings {
      bUseBrushTransparency=false
      roundingType=HalfHeightRadius
      width=0
      color {colorUseRule=UseColor_Specified, specifiedColor=<rgba 0 0 0 0>}
      cornerRadii=<0 0 0 0>
     }
     tintColor {colorUseRule=UseColor_Specified, specifiedColor=<rgba 1 1 1 1>}
     uVRegion {bIsValid=false, max=<0 0>, min=<0 0>}
    }}
   slot {class=OverlaySlot}
  }
  VerticalBox_1 {
   widgetClassPath=&VerticalBox
   parent=Overlay_Container
   widget=VerticalBox_1
   slot {class=OverlaySlot, properties {verticalAlignment=VAlign_Center, padding=<ltrb 140 0 0 0>}}
  }
  CommonTextBlock_97 {
   widgetClassPath=&CommonTextBlock
   parent=VerticalBox_1
   widget=CommonTextBlock_97
   properties {
    text=with
    textTransformPolicy=ToUpper
    font {
     bForceMonospaced=false
     bMaterialIsStencil=false
     fontMaterial=null
     fontObject=&Burbank
     letterSpacing=0
     monospacedWidth=1
     size=20
     skewAmount=0
     typefaceFontName=Burbank-Black
     outlineSettings {
      bApplyOutlineToDropShadows=true
      bMiteredCorners=false
      bSeparateFillAlpha=true
      outlineMaterial=&M-FontShadow
      outlineSize=3
      outlineColor=<rgba 0 0 0 0.5>
     }
    }
    shadowOffset=<0 0>
   }
   slot {class=VerticalBoxSlot, properties {horizontalAlignment=HAlign_Left, padding=<ltrb 0 -5 0 6>}}
  }
  SizeBox_2 {
   widgetClassPath=&SizeBox
   parent=VerticalBox_1
   widget=SizeBox_2
   properties {bOverride_MaxDesiredHeight=true, bOverride_WidthOverride=true, maxDesiredHeight=80, widthOverride=345.8244}
   slot {class=VerticalBoxSlot, properties {horizontalAlignment=HAlign_Left, padding=<ltrb 0 -15 0 -5>}}
  }
  ScaleBox_0 {widgetClassPath=&ScaleBox, parent=SizeBox_2, widget=ScaleBox_0, properties {stretchDirection=DownOnly}, slot {class=SizeBoxSlot}}
  Text_WeaponName {
   widgetClassPath=&CommonTextBlock
   bIsVariable=true
   parent=ScaleBox_0
   widget=Text_WeaponName
   properties {
    lineHeightPercentage=0.9
    text=Weapon Name
    textTransformPolicy=ToUpper
    font {
     bForceMonospaced=false
     bMaterialIsStencil=false
     fontMaterial=null
     fontObject=&Burbank
     letterSpacing=0
     monospacedWidth=1
     size=31
     skewAmount=0
     typefaceFontName=Burbank-Black
     outlineSettings {
      bApplyOutlineToDropShadows=true
      bMiteredCorners=false
      bSeparateFillAlpha=true
      outlineMaterial=&M-FontShadow
      outlineSize=3
      outlineColor=<rgba 0 0 0 0.5>
     }
    }
    shadowOffset=<0 0>
   }
   slot {class=ScaleBoxSlot, properties {horizontalAlignment=HAlign_Left}}
  }
  Text_WeaponRarity {
   widgetClassPath=&CommonTextBlock
   bIsVariable=true
   parent=VerticalBox_1
   widget=Text_WeaponRarity
   properties {
    text="{Rarity}"
    textTransformPolicy=ToUpper
    font {
     bForceMonospaced=false
     bMaterialIsStencil=false
     fontMaterial=null
     fontObject=&Burbank
     letterSpacing=0
     monospacedWidth=1
     size=21
     skewAmount=0
     typefaceFontName=BurbankRegular-Black
     outlineSettings {
      bApplyOutlineToDropShadows=false
      bMiteredCorners=false
      bSeparateFillAlpha=false
      outlineMaterial=null
      outlineSize=0
      outlineColor=<rgba 0 0 0 1>
     }
    }
    shadowOffset=<0 0>
   }
   slot {class=VerticalBoxSlot, properties {horizontalAlignment=HAlign_Left, padding=<ltrb 0 3 0 0>}}
  }
 ]
}
