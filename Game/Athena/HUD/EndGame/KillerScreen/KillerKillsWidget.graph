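# KillerKillsWidget is a UMG widget for the end-game killer screen that shows a kill count as an icon, spacer and number in a horizontal row.
# PreConstruct applies its variables for text color, text size, icon size and gap, rebuilding the font struct to change only the size.
# The Set Num Of Kills event makes the widget HitTestInvisible and writes the count as text, capped at three integral digits.
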WidgetBlueprint /Game/Athena/HUD/EndGame/KillerScreen/KillerKillsWidget
graphs [EventGraph {nodes [
   n1 = Set Num Of Kills!() -> (OutputDelegate, then, Num Of Kills:int) {is=title K2Node_CustomEvent}
   > SetVisibility(InVisibility=HitTestInvisible)
   > "SetText (Text)"(
    self=Get Text_NumOfKills() -> (Text_NumOfKills:Common Text Object Reference)
    InText="To Text (Integer)"(Value=@n1.Num Of Kills, MaximumIntegralDigits=3)
   )
   Event PreConstruct()
   > n2 = Sequence!(execute) -> (then_0, then_1, then_2:Exec, then_3:Exec)
   > SetColorAndOpacity(self=Get Text_NumOfKills(), InColorAndOpacity=Get Text Color and Opacity() -> (Text Color and Opacity:Slate Color Structure))
   n3 = Get Text_NumOfKills()
   n4 = Break Slate Font Info!(SlateFontInfo:"Slate Font Info Structure (by ref)"=Get Font!(self:Text Object Reference=@n3) -> (Font:Slate Font Info Structure)) -> (
    FontObject:Object Reference
    FontMaterial:Object Reference
    OutlineSettings:Font Outline Settings Structure
    TypefaceFontName:Name
    Size:float=~
    LetterSpacing:int
    SkewAmount:float=~
    bForceMonospaced:bool=~
    bMaterialIsStencil:bool=~
    MonospacedWidth:float=~
   ) {is=title K2Node_BreakStruct}
   SetFont(
    execute=@n2.then_1
    self=@n3
    InFontInfo=MakeSlateFontInfo!(
     FontObject:Object Reference=@n4
     FontMaterial:Object Reference=@n4.FontMaterial
     OutlineSettings:Font Outline Settings Structure=@n4.OutlineSettings
     TypefaceFontName:Name=@n4.TypefaceFontName
     Size:float=Get Text Size() -> (Text Size:float)
     LetterSpacing:int=@n4.LetterSpacing
     SkewAmount:float=@n4.SkewAmount
     bForceMonospaced:bool=@n4.bForceMonospaced
     bMaterialIsStencil:bool=@n4.bMaterialIsStencil
     MonospacedWidth:float=@n4.MonospacedWidth
    ) -> (ReturnValue:Slate Font Info Structure)
   )
   n5 = Get Icon Size() -> (Icon Size:double)
   SetDesiredSizeOverride!(
    execute=@n2.then_2
    self:Image Object Reference=Get Icon_Kills() -> (Icon_Kills:Image Object Reference)
    DesiredSize:Vector 2D Structure=MakeVector2D(X=@n5, Y=@n5)
   ) -> (then)
   SetSize!(
    execute=@n2.then_3
    self:Spacer Object Reference=Get Spacer() -> (Spacer:Spacer Object Reference)
    InSize:Vector 2D Structure=MakeVector2D(X=Get Space() -> (Space:double), Y=1)
   ) -> (then)
   Event Construct()
   Event Tick()
  ]}]
variables [
 Icon Size {type=number, default=64}
 Space {type=number, default=16}
 Text Color and Opacity {type=SlateColor, default {colorUseRule=UseColor_Specified, specifiedColor=<rgba 1 1 1 1>}}
 Text Size {type=number, default=57.33}
]
widgets {
 info {inheritedWidgetCount=0, parentClass=&UserWidget, rootWidgetClass=&HorizontalBox, widgetCount=4}
 widgets [
  HorizontalBox_13 {widgetClassPath=&HorizontalBox, widget=HorizontalBox_13}
  Icon_Kills {
   widgetClassPath=&Image
   bIsVariable=true
   parent=HorizontalBox_13
   widget=Icon_Kills
   properties {brush {
     bIsDynamicallyLoaded=false
     drawAs=Image
     imageType=NoImage
     mirroring=NoMirror
     resourceName=None
     resourceObject=/Game/Athena/HUD/Art/killsico_alt.KillsIco_Alt
     tiling=NoTile
     imageSize=<64 64>
     margin=<ltrb 0 0 0 0>
     outlineSettings {
      bUseBrushTransparency=false
      roundingType=HalfHeightRadius
      width=0
      color {colorUseRule=UseColor_Specified, specifiedColor=<rgba 0 0 0 0>}
      cornerRadii=<0 0 0 0>
     }
     tintColor {colorUseRule=UseColor_Specified, specifiedColor=<rgba 1 1 1 1>}
     uVRegion {bIsValid=false, max=<0 0>, min=<0 0>}
    }}
   slot {class=HorizontalBoxSlot, properties {horizontalAlignment=HAlign_Center, verticalAlignment=VAlign_Center}}
  }
  Spacer {
   widgetClassPath=&Spacer
   bIsVariable=true
   parent=HorizontalBox_13
   widget=Spacer
   properties {size=<16 1>}
   slot {class=HorizontalBoxSlot}
  }
  Text_NumOfKills {
   widgetClassPath=&CommonTextBlock
   bIsVariable=true
   parent=HorizontalBox_13
   widget=Text_NumOfKills
   properties {
    minDesiredWidth=0.4
    text=99
    font {
     bForceMonospaced=false
     bMaterialIsStencil=false
     fontMaterial=null
     fontObject=&Burbank
     letterSpacing=0
     monospacedWidth=1
     size=43
     skewAmount=0
     typefaceFontName=Burbank-Black
     outlineSettings {
      bApplyOutlineToDropShadows=false
      bMiteredCorners=false
      bSeparateFillAlpha=false
      outlineMaterial=null
      outlineSize=0
      outlineColor=<rgba 0 0 0 1>
     }
    }
    shadowColorAndOpacity=<rgba 0.0052 0.0052 0.0056 0>
    shadowOffset=<0 0>
   }
   slot {class=HorizontalBoxSlot, properties {verticalAlignment=VAlign_Center}}
  }
 ]
}
